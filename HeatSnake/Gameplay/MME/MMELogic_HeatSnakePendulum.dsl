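; Movement mode extension logic for the HeatSnake pendulum swing, active while the pawn has the swing tag and a tether point.
; Physics update first launches the pawn toward the tether at distance-based, capped speed, then pushes along the cross product of tether and camera right vector, scaling acceleration by swing angle.
; Rising above the tether, or overstretching on non-servers, severs it.
; Deactivation clears state and tags, applying fall damage immunity.
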(event HeatSnake|EventRecordInput(De-Nativized)
  (params (RuntimeContext :type MMERuntime%20Context%20Structure%20%28by%20ref%29)))

(event EventOnExtensionDeactivate
  (params (RuntimeContext :type MMERuntime%20Context%20Structure%20%28by%20ref%29))
  (|CallParentFunction RuntimeContext)
  (GameplayEffects|ApplyGameplayEffectToTarget :self (Ability|GetAbilitySystemComponent (Variables|FortMovementMode_BaseExtLogic|GetOwnerPawn)) :GameplayEffectClass (Variables|Default|GetGE_FallDamageImmunity) :Target (Ability|GetAbilitySystemComponent (Variables|FortMovementMode_BaseExtLogic|GetOwnerPawn)) :Level 0.0)
  (CallFunction|CleanUpMME RuntimeContext)
  (GameplayCue|ExecuteGameplayCueOnActor(Burst) (Variables|FortMovementMode_BaseExtLogic|GetOwnerPawn) (Variables|Default|GetMMEEndCue) (Ability|GameplayCue|MakeGameplayCueParameters 0.0)))

(event EventOnExtensionActivate
  (params (RuntimeContext :type MMERuntime%20Context%20Structure%20%28by%20ref%29))
  (|CallParentFunction RuntimeContext)
  (CallFunction|ClearTransientLogic)
  (GameplayCue|ExecuteGameplayCueOnActor(Burst) (Variables|FortMovementMode_BaseExtLogic|GetOwnerPawn) (Variables|Default|GetMMEStartCue) (Ability|GameplayCue|MakeGameplayCueParameters 0.0)))

(event EventOnExtensionPush
  (params (RuntimeContext :type MMERuntime%20Context%20Structure%20%28by%20ref%29))
  (|CallParentFunction RuntimeContext)
  (CallFunction|ClearTransientLogic))

(fn PhysUpdate(De-Nativized)
  (params (RuntimeContext :type MMERuntime%20Context%20Structure%20%28by%20ref%29) (InVelocity :type Vector%20%28by%20ref%29 :default "0, 0, 0") (InAcceleration :type Vector%20%28by%20ref%29 :default "0, 0, 0"))
  (bind _cachedownerpawn (Variables|Transient|GetCachedOwnerPawn))
  (bind _returnvalue (Transformation|GetActorLocation _cachedownerpawn))
  (bind _cachedownerpawn_2 (Variables|Transient|GetCachedOwnerPawn))
  (bind _returnvalue_2 (Pawn|GetController _cachedownerpawn_2))
  (bind _returnvalue_3 (Pawn|GetDesiredRotation _returnvalue_2))
  (bind _returnvalue_4 (Math|Vector|GetRightVector _returnvalue_3))
  (bind _maxtetherdistsquared (Variables|Transient|GetMaxTetherDistSquared))
  (bind _shouldsever (CallFunction|CheckSeverPendulumDuetoDistTether _maxtetherdistsquared))
  (bind _runtimecontext (Variables|GetRuntimeContext))
  (bind _returnvalue_5 (Class|FortMovementModeExtBlueprintLibrary|GetDeltaTimefromContext _runtimecontext))
  (bind _tetherpoint (Variables|Transient|GetTetherPoint))
  (bind _pawnlocation (LocalVariables|GetPawnLocation))
  (bind _returnvalue_6 (- _pawnlocation _tetherpoint))
  (bind _returnvalue_7 (Math|Vector|Normalize _returnvalue_6))
  (bind _cachedownerpawn_3 (Variables|Transient|GetCachedOwnerPawn))
  (bind _returnvalue_8 (Transformation|GetActorLocation _cachedownerpawn_3))
  (bind _bisserver (Variables|FortMovementMode_BaseExtLogic|GetbIsServer))
  (bind _returnvalue_9 (not _bisserver))
  (bind _hf_percentoftetherdistanceshouldbeinitiallaunch (Variables|Hotfix|GetHF_PercentOfTetherDistanceShouldBeInitialLaunch))
  (bind _returnvalue_10 (ScalableFloat|GetCachedScalableFloatValue _hf_percentoftetherdistanceshouldbeinitiallaunch))
  (bind _returnvalue_11 (Math|Float|Clamp(Float) _returnvalue_10))
  (bind _returnvalue_12 (and _returnvalue_9 _shouldsever))
  (bind _tetherpoint_2 (Variables|Transient|GetTetherPoint))
  (bind _returnvalue_13 (- _tetherpoint_2 _returnvalue_8))
  (bind _initialtetherdistance (Variables|Transient|GetInitialTetherDistance))
  (bind _returnvalue_14 (* _returnvalue_11 _initialtetherdistance))
  (bind _mmeactivated_tags (Variables|Default|GetMMEActivated%20Tags))
  (bind _runtimedata (LocalVariables|GetRuntimeData))
  (bind _cachedownerpawn_4 (Variables|Transient|GetCachedOwnerPawn))
  (bind _pawnlocation_2 (LocalVariables|GetPawnLocation))
  (bind _normalizedcrossproduct (LocalVariables|GetNormalizedCrossProduct))
  (LocalVariables|SetResultingVelocity InVelocity)
  (LocalVariables|SetResultingAcceleration InAcceleration)
  (LocalVariables|GetRuntimeData
    (:then
      _)
    (:else
      (bind _returnvalue_15 (Class|FortMovementModeExtBlueprintLibrary|GetRuntimeDatafromContext RuntimeContext "/Script/HeatSnakeRuntime.FortMovementMode_HeatSnakePendulumRuntimeData"))
      (LocalVariables|SetRuntimeData _returnvalue_15)))
  (bind _ownerpawn (Variables|FortMovementMode_BaseExtLogic|GetOwnerPawn)
    (:then
      (Variables|Transient|SetCachedOwnerPawn _ownerpawn)
      (if (not (Math|Vector|VectorIsZero (Variables|Transient|GetTetherPoint)))
        (Utilities|FlowControl|Sequence
          (:then_0
            ; Launch toward the tether
            (if (<= (Variables|/Script/HeatSnakeRuntime.FortMovementMode_HeatSnakePendulumRuntimeData|GetTimeElapsed (LocalVariables|GetRuntimeData)) (ScalableFloat|GetValueatLevel (Variables|Hotfix|GetHF_LaunchTowardTetherDuration)))
              (if (Math|Vector|VectorIsZero (Variables|Transient|GetLaunchDirectionVector))
                ; Calculate Initial Launch Towards Target
                (Variables|Transient|SetInitialTetherDistance (Math|Vector|VectorLength _returnvalue_13))
                (Variables|Transient|SetLaunchDirectionVector (Math|Vector|Normalize _returnvalue_13 0.001))
                (Variables|Transient|SetFixedLaunchDistance (+ (* (ScalableFloat|GetValueatLevel (Variables|Hotfix|GetHF_InitialLaunchAcceleration)) (Math|Float|Square (ScalableFloat|GetValueatLevel (Variables|Hotfix|GetHF_LaunchTowardTetherDuration)))) (* (ScalableFloat|GetValueatLevel (Variables|Hotfix|GetHF_InitialLaunchVelocity)) (ScalableFloat|GetValueatLevel (Variables|Hotfix|GetHF_LaunchTowardTetherDuration))))) ; Distance our player will always travel with our fixed forces
                (Utilities|FlowControl|Sequence
                  (:then_0
                    (if (> _returnvalue_14 (Variables|Transient|GetFixedLaunchDistance))
                      ; Calculate the Hijacked Velocity (divides are expensive so separate by exec pin)
                      ; Cap the Final Launch Velocity so Player doesn't end up in orbit on incidental Tap inputs
                      (Variables|Transient|SetFinalLaunchVelocity (select (> (Math|Float|SafeDivide _returnvalue_14 (ScalableFloat|GetValueatLevel (Variables|Hotfix|GetHF_LaunchTowardTetherDuration))) (ScalableFloat|GetValueatLevel (Variables|Hotfix|GetHF_MaxVelocitySpeedLimit))) (ScalableFloat|GetValueatLevel (Variables|Hotfix|GetHF_MaxVelocitySpeedLimit)) (Math|Float|SafeDivide _returnvalue_14 (ScalableFloat|GetValueatLevel (Variables|Hotfix|GetHF_LaunchTowardTetherDuration)))))
                      (else
                        ; Keep our fixed Velocity
                        (Variables|Transient|SetFinalLaunchVelocity (ScalableFloat|GetValueatLevel (Variables|Hotfix|GetHF_InitialLaunchVelocity))))))
                  (:then_1
                    _))
                (elif (>= (/ (Math|Vector|Distance(Vector) (Transformation|GetActorLocation (Variables|Transient|GetCachedOwnerPawn)) (Variables|Transient|GetTetherPoint)) (Variables|Transient|GetInitialTetherDistance)) (Math|Float|Clamp(Float) (ScalableFloat|GetCachedScalableFloatValue (Variables|Hotfix|GetHF_PercentOfTetherDistanceShouldBeInitialLaunch))))
                  _
                  (else
                    (bind _output_get (LocalVariables|SetPawnLocation (Transformation|GetActorLocation (Variables|Transient|GetCachedOwnerPawn)) :!label _label))
                    (bind (_output_get_x _output_get_y _output_get_z) (Math|Vector|BreakVector _output_get))
                    (if (<= (Variables|Transient|GetMaxTetherDistSquared) 0)
                      ; Make the new tether length after launching toward the tether point
                      (Variables|Transient|SetMaxTetherDistSquared (* (Math|Vector|DistanceSquared(Vector) (LocalVariables|GetPawnLocation) (Variables|Transient|GetTetherPoint)) 2.0)))
                    (if (or (CallFunction|CheckSeverPendulumDuetoHeightAboveTether _output_get_z) _returnvalue_12)
                      ; Sever tether connection and let the player fling
                      (if _returnvalue_12
                        (Rendering|Debug|DrawDebugLine (Variables|Transient|GetCachedOwnerPawn) (LocalVariables|GetPawnLocation) (Variables|Transient|GetTetherPoint) "(R=1.000000,G=0.429277,B=0.000000,A=1.000000)" 2e+01 2.0)
                        (else
                          (Rendering|Debug|DrawDebugLine (Variables|Transient|GetCachedOwnerPawn) (LocalVariables|GetPawnLocation) (Variables|Transient|GetTetherPoint) "(R=0.329156,G=0.000000,B=1.000000,A=1.000000)" 2e+01 2.0)))
                      (Ability|GameplayEffect|RemoveGameplayTags (Variables|FortMovementMode_BaseExtLogic|GetOwnerPawn) _mmeactivated_tags)
                      (GameplayEffects|RemoveActiveEffectswithGrantedTags (Variables|FortMovementMode_BaseExtLogic|GetOwnerAbilitySystemComponent) _mmeactivated_tags)
                      (CallFunction|CleanUpMME (Variables|GetRuntimeContext))
                      (else
                        ; Add force along pendulum using camera direction
                        ;; Find direction of pendulum's force by getting the normal of the plane created by the Camera's RightVector and the tether to the player
                        (Utilities|IsValid _returnvalue_2
                          (:Is%20Valid
                            (LocalVariables|SetNormalizedCrossProduct (Math|Vector|Normalize (Math|Vector|CrossProduct _returnvalue_7 _returnvalue_4) 0.001))
                            ;; Debug Lines for showing pendulum Axis
                            (Rendering|Debug|DrawDebugLine _cachedownerpawn_4 _pawnlocation_2 (+ (* _returnvalue_4 2e+02) (LocalVariables|GetPawnLocation)) "(R=1.000000,G=0.000000,B=0.818404,A=1.000000)" 2e+01 0.5)
                            (Rendering|Debug|DrawDebugLine _cachedownerpawn_4 _pawnlocation_2 (+ (* _returnvalue_7 2e+02) (LocalVariables|GetPawnLocation)) "(R=1.000000,G=0.814955,B=0.000000,A=1.000000)" 2e+01 0.5)
                            (Rendering|Debug|DrawDebugLine _cachedownerpawn_4 _pawnlocation_2 (+ (* (LocalVariables|GetNormalizedCrossProduct) 2e+02) (LocalVariables|GetPawnLocation)) "(R=0.075942,G=1.000000,B=0.001513,A=1.000000)" 2e+01 0.5)
                            (LocalVariables|SetResultingVelocity (* _normalizedcrossproduct (ScalableFloat|GetValueatLevel (Variables|Hotfix|GetHF_AdditionalSwingVelocity))))
                            (LocalVariables|SetResultingAcceleration (* _normalizedcrossproduct (* (ScalableFloat|GetValueatLevel (Variables|Hotfix|GetHF_AdditionalSwingAcceleration)) (Math|Trig|Sin(Radians) (Math|Trig|Acos(Radians) (Math|Vector|DotProduct (Math|Vector|Normalize (- (LocalVariables|GetPawnLocation) (Variables|Transient|GetTetherPoint))) (Math|Vector|VectorDown))))))))))))))
              ; Keep Launching toward the tetherpoint
              (LocalVariables|SetResultingVelocity (* (Variables|Transient|GetLaunchDirectionVector) (Variables|Transient|GetFinalLaunchVelocity)))
              ;; For now it's too complicated to also adjust proportionate acceleration so we accept the small discrepency
              (LocalVariables|SetResultingAcceleration (* (Variables|Transient|GetLaunchDirectionVector) (ScalableFloat|GetValueatLevel (Variables|Hotfix|GetHF_InitialLaunchAcceleration)))))
            (goto _label))
          (:then_1
            (Variables|/Script/HeatSnakeRuntime.FortMovementMode_HeatSnakePendulumRuntimeData|SetTimeElapsed (+ _returnvalue_5 (Variables|/Script/HeatSnakeRuntime.FortMovementMode_HeatSnakePendulumRuntimeData|GetTimeElapsed (LocalVariables|GetRuntimeData))) _runtimedata)
            (Variables|/Script/HeatSnakeRuntime.FortMovementMode_HeatSnakePendulumRuntimeData|SetLastPosition (Transformation|GetActorLocation (Variables|Transient|GetCachedOwnerPawn)) _runtimedata))
          (:then_2
            (Rendering|Debug|DrawDebugLine _cachedownerpawn _returnvalue (+ _returnvalue (* _returnvalue_5 (LocalVariables|GetResultingVelocity))) "(R=1.000000,G=0.000000,B=0.034817,A=1.000000)" 2e+01 1.0) ; Debug Velocity
            (Rendering|Debug|DrawDebugLine _cachedownerpawn _returnvalue (+ _returnvalue (* _returnvalue_5 (LocalVariables|GetResultingAcceleration))) "(R=0.002815,G=0.016582,B=1.000000,A=1.000000)" 2e+01 1.0) ; Debug Acceleration
            (Rendering|Debug|DrawDebugSphere _cachedownerpawn (Variables|Transient|GetTetherPoint) 1e+02 12 "(R=1.000000,G=0.000000,B=0.044378,A=1.000000)" 2e+01 0.5) ; Debug TetherPoint
            (return (LocalVariables|GetResultingVelocity) (LocalVariables|GetResultingAcceleration))))
        (else
          (Development|PrintString 0 "HeatSnakeTetherpoint is Zero"))))
    (:else
      (Fortnite|FortShippingLog self "PendulumMMD: Owner Pawn NOT valid")
      (return (LocalVariables|GetResultingVelocity) (LocalVariables|GetResultingAcceleration)))))

(fn UpdateBeforeCharacterMovement(De-Nativized)
  (params (RuntimeContext :type MMERuntime%20Context%20Structure%20%28by%20ref%29))
  (bind _ownerpawn (Variables|FortMovementMode_BaseExtLogic|GetOwnerPawn)
    (:then
      (if (GameplayTags|HasMatchingGameplayTag _ownerpawn "(TagName=\"HeatSnake.Ability.PendulumSwing\")")
        (bind _returnvalue (Class|FortMovementModeExtBlueprintLibrary|GetRuntimeDatafromContext RuntimeContext "/Script/HeatSnakeRuntime.FortMovementMode_HeatSnakePendulumRuntimeData"))
        (if (and (Variables|/Script/HeatSnakeRuntime.FortMovementMode_HeatSnakePendulumRuntimeData|GetbSecondaryReleased _returnvalue) (not (Math|Vector|VectorIsZero (Variables|Transient|GetTetherPoint))))
          _
          (elif (Networking|HasAuthority _ownerpawn)
            (if (Networking|HasAuthority _ownerpawn) :!label _label
              (Variables|Transient|SetTetherPoint (Variables|/Script/HeatSnakeRuntime.FortMovementMode_HeatSnakePendulumRuntimeData|GetInitialTetherPoint _returnvalue))
              (else
                (Variables|Transient|SetTetherPoint (Variables|FortMovementMode_HeatSnakePendulumExtLogic|GetCachedTetherPoint))))
            (return "Active")
            (elif (Math|Vector|VectorIsNearlyZero (Variables|/Script/HeatSnakeRuntime.FortMovementMode_HeatSnakePendulumRuntimeData|GetLastPosition _returnvalue) 0.001)
              (goto _label)
              (elif (> (Math|Vector|DistanceSquared(Vector) (Transformation|GetActorLocation (Variables|FortMovementMode_BaseExtLogic|GetOwnerPawn)) (Variables|/Script/HeatSnakeRuntime.FortMovementMode_HeatSnakePendulumRuntimeData|GetLastPosition _returnvalue)) 0.001)
                (goto _label)
                (else
                  (Rendering|Debug|DrawDebugLine (Variables|FortMovementMode_BaseExtLogic|GetOwnerPawn) (Transformation|GetActorLocation (Variables|FortMovementMode_BaseExtLogic|GetOwnerPawn)) (Variables|Transient|GetTetherPoint) "(R=0.277896,G=0.841032,B=1.000000,A=1.000000)" 99999.0 1.0) ; Debug Acceleration
                )))))
        (else
          (Development|PrintString 0 (Utilities|String|SelectString "Server TagMissing" "Client TagMissing" (Networking|HasAuthority _ownerpawn)))
          (Rendering|Debug|DrawDebugLine (Variables|FortMovementMode_BaseExtLogic|GetOwnerPawn) (Transformation|GetActorLocation (Variables|FortMovementMode_BaseExtLogic|GetOwnerPawn)) (Variables|Transient|GetTetherPoint) "(R=1.000000,G=0.917599,B=0.795168,A=1.000000)" 99999.0 1.0) ; Debug Acceleration
        )))
    (:else
      _))
  (return "Ended"))

(fn CheckSeverPendulumDueToDistTether
  (params (MaxTetherDistSquared :type Float%20%28double-precision%29))
  (return (< MaxTetherDistSquared (Math|Vector|DistanceSquared(Vector) (Variables|Transient|GetTetherPoint) (Transformation|GetActorLocation (Variables|Transient|GetCachedOwnerPawn))))))

(fn CheckSeverPendulumDueToHeightAboveTether
  (params (OwningPawnHeight :type Float%20%28double-precision%29))
  (bind (_tetherpoint_x _tetherpoint_y _tetherpoint_z) (Math|Vector|BreakVector (Variables|GetTetherPoint_X)))
  (return (> OwningPawnHeight _tetherpoint_z)))

(fn CleanUpMME
  (params (RuntimeContext :type MMERuntime%20Context%20Structure%20%28by%20ref%29))
  (bind _mmeactivated_tags (Variables|Default|GetMMEActivated%20Tags))
  ; Clear these because the first few updates may otherwise resuse an old tether and launch us the wrong way
  (CallFunction|ClearTransientLogic)
  (bind _returnvalue (Class|FortMovementModeExtBlueprintLibrary|GetRuntimeDatafromContext RuntimeContext "/Script/HeatSnakeRuntime.FortMovementMode_HeatSnakePendulumRuntimeData"))
  (Utilities|IsValid _returnvalue
    (:Is%20Valid
      (Variables|/Script/HeatSnakeRuntime.FortMovementMode_HeatSnakePendulumRuntimeData|SetTimeElapsed 0.0 _returnvalue))
    (:Is%20Not%20Valid
      _))
  (bind _ownerpawn (Variables|FortMovementMode_BaseExtLogic|GetOwnerPawn)
    (:then
      (Ability|GameplayEffect|RemoveGameplayTags _ownerpawn _mmeactivated_tags)
      (GameplayEffects|RemoveActiveEffectswithGrantedTags (Variables|FortMovementMode_BaseExtLogic|GetOwnerAbilitySystemComponent) _mmeactivated_tags))
    (:else
      (Development|PrintString 0 "HeatSnake cleanup error. Player is Invalid!" true true "(R=1.000000,G=0.000000,B=0.090303,A=1.000000)"))))

(fn ClearTransientLogic
  (Variables|Transient|SetTetherPoint "0, 0, 0")
  (Variables|Transient|SetLaunchDirectionVector "0, 0, 0")
  (Variables|Transient|SetMaxTetherDistSquared 0.0)
  (Variables|Transient|SetInitialTetherDistance 0.0)
  (Variables|Transient|SetFinalLaunchVelocity 0.0))
